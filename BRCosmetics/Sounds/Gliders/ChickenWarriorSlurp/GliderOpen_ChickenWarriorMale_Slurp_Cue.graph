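# Sound cue for the glider-open sound of the male Chicken Warrior Slurp cosmetic glider.
# A mixer layers the main open wave with a slightly quieter caw, which fires after a short random delay and gets random pitch and volume modulation.
# It primes on load, uses glider attenuation and property sheet, and overrides its concurrency set with a small per-owner limit that stops the oldest.
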SoundCue /BRCosmetics/Sounds/Gliders/ChickenWarriorSlurp/GliderOpen_ChickenWarriorMale_Slurp_Cue
root=Mixer_0
nodes [
 {id=Delay_0, children [WavePlayer_0], properties {delayMax=2, delayMin=1}}
 {id=Mixer_0, children [WavePlayer_3,Modulator_1], properties {inputVolume [1,0.8]}}
 {id=Modulator_1, children [Delay_0], properties {pitchMax=1.1, volumeMin=1}}
 {id=WavePlayer_0, properties {soundWaveAssetPtr=&/GliderOpen_Chicken_Warrior_Slurp_Male_Caw}}
 {id=WavePlayer_3, properties {soundWaveAssetPtr=&/GliderOpen_Chicken_Warrior_Slurp_Male}}
]
properties {
 attenuationSettings=&Fort_Player_Glider_Thrust_Athena_Att
 audioPropertiesSheet=/BRCosmetics/Sounds/Settings/PropertySheets/Glider/APS_Cosmetic_Glider_Open
 bOverrideConcurrency=true
 bPrimeOnLoad=true
 maxDistance=12500
 soundClassObject=&Fort_Gliders_Land
 volumeMultiplier=0.8
 concurrencyOverrides {bLimitToOwner=true, maxCount=2, resolutionRule=StopOldest}
 concurrencySet [&Athena_GliderOpenClose_Conc]
}
